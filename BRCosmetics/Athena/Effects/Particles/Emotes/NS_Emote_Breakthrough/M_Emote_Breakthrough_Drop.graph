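# Masked, lit Niagara sprite material for drop particles in the Breakthrough emote effect.
# Fakes a sphere on each sprite by rebuilding a normal from centered UVs, with a circular cutout mask.
# A fixed camera-space light direction against that normal yields a highlight with adjustable size and softness, tinted by particle color into emissive.
# Particle alpha drives a soft diagonal UV wipe, multiplied by the sphere depth.
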Material /BRCosmetics/Athena/Effects/Particles/Emotes/NS_Emote_Breakthrough/M_Emote_Breakthrough_Drop
graph {
 properties {bUsedWithNiagaraSprites=true, blendMode=BLEND_Masked, pixelDepthOffsetMode=PDOM_Legacy, usedShadingModels=Default Lit}
 nodes [
  n1 = ConstantBiasScale(None=TextureCoordinate()) {bias=-0.5, scale=2}
  n2 = OneMinus(None=Saturate(None=DotProduct(A=@n1, B=@n1)))
  Ceil(None=@n2)
  Constant()
  Constant() {r=1}
  Constant() {r=0.02}
  n3 = Normalize(VectorInput=Transform(None=Constant3Vector() {constant=<rgba -0.3 0.6 -1 0>}) {transformSourceType=TRANSFORMSOURCE_Camera, transformType=TRANSFORM_Tangent})
  n4 = SquareRoot(None=@n2)
  n5 = AppendVector(A=@n1, B=@n4)
  n6 = Saturate(None=ScalarParameter() {defaultValue=0.05, parameterName=Highlight Size, sliderMax=1})
  n7 = Divide(
   A=Subtract(A=DotProduct(A=@n3, B=@n5), B=OneMinus(None=@n6))
   B=Saturate(None=ScalarParameter() {defaultValue=0.1, parameterName=Highlight Softness, sliderMax=1})
  )
  MaterialFunctionCall(Emissive=Multiply(A=ParticleColor().RGB, B=Saturate(None=@n7))) {materialFunction=&MF_FX_EmissiveAdjustments}
  n8 = DotProduct(A=TextureCoordinate(), B=Constant() {r=0.5})
  n9 = OneMinus(None=Saturate(None=ParticleColor().A))
  n10 = Subtract(A=OneMinus(None=Frac(None=@n8)), B=Multiply(A=@n9, B=@n9))
  Multiply(A=@n4, B=Saturate(None=Divide(A=@n10, B=@n9) {constB=0.3}))
 ]
}
