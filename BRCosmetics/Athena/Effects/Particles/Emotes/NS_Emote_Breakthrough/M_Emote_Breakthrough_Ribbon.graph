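# Masked material for Niagara ribbons and sprites in the Breakthrough emote effect, built on a horizontally scrolling ribbon texture.
# Texture red and green channels become a normal, dotted with a fixed camera-space direction in tangent space to fake a highlight with adjustable size and softness.
# That highlight, tinted by particle color, feeds an emissive adjustment function.
# The blue channel is hard-thresholded against particle alpha, forming an erosion-style mask.
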Material /BRCosmetics/Athena/Effects/Particles/Emotes/NS_Emote_Breakthrough/M_Emote_Breakthrough_Ribbon
graph {
 properties {
  bUsedWithNiagaraRibbons=true
  bUsedWithNiagaraSprites=true
  blendMode=BLEND_Masked
  pixelDepthOffsetMode=PDOM_Legacy
  usedShadingModels=Default Lit
 }
 nodes [
  n1 = TextureSample(UVs=Panner() {speedX=3}) {samplerType=SAMPLERTYPE_LinearColor, texture=&/T_Emote_Breakthrough_Ribbon}
  n2 = Saturate(None=ParticleColor().A)
  Ceil(None=Subtract(A=@n1.B, B=OneMinus(None=@n2)))
  Constant()
  Constant() {r=1}
  Constant() {r=0.02}
  n3 = Normalize(VectorInput=Transform(None=Constant3Vector() {constant=<rgba -0.3 0.6 -1 0>}) {transformSourceType=TRANSFORMSOURCE_Camera, transformType=TRANSFORM_Tangent})
  n4 = DeriveNormalZ(InXY=ConstantBiasScale(None=ComponentMask(None=@n1.RGB) {g=true, r=true}) {bias=-0.5, scale=2})
  n5 = Saturate(None=ScalarParameter() {defaultValue=0.05, parameterName=Highlight Size, sliderMax=1})
  n6 = Divide(
   A=Subtract(A=DotProduct(A=@n3, B=@n4), B=OneMinus(None=@n5))
   B=Saturate(None=ScalarParameter() {defaultValue=0.1, parameterName=Highlight Softness, sliderMax=1})
  )
  MaterialFunctionCall(Emissive=Multiply(A=ParticleColor().RGB, B=Saturate(None=@n6))) {materialFunction=&MF_FX_EmissiveAdjustments}
 ]
}
