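# Niagara system for the Breakthrough emote that plays a small liquid splash with droplets and a thin trailing stream.
# Splash and Droplets_Splash burst sprites along angled cones, growing and fading the splash while shrinking the droplets.
# After a short delay, a brief ribbon stream spawns alongside velocity-inheriting droplets, followed later by sparse residual drips.
# Gravity pulls droplets and ribbon down, drag slows the sprites, and curves drive size and alpha over life.
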NiagaraSystem /BRCosmetics/Athena/Effects/Particles/Emotes/NS_Emote_Breakthrough/NS_Emote_Breakthrough
dependencies {usedDataInterfaces [&NiagaraDataInterfaceCurve], usedDynamicInputs [], usedModules [&SystemState]}
emitters {
 Droplets_Residual {
  emitterUpdateScript [
   &EmitterState(Life Cycle Mode=::NewEnumerator1, Loop Behavior=::NewEnumerator1, Loop Delay=1.75, Loop Duration=0.3, UseLoopDelay=true)
   &SpawnRate(Spawn Probability=0.6, SpawnRate=13, Use Spawn Probability=true)
  ]
  particleSpawnScript [&InitializeParticle(
    Color Mode=::NewEnumerator0
    Lifetime=&RandomRangeFloat(Minimum=0.333, Maximum=1)
    Sprite Size=<1.1 2>
    Sprite Size Mode=::NewEnumerator4
    UsePositionOffset=false
    Write Material Random=true
   )]
  particleUpdateScript [
   &GravityForce(Gravity=&RandomRangeVector(Minimum=<0 0 -444>, Maximum=<1 1 -666>))
   &Drag(Drag=3, Use Linear Drag=true, Use Rotational Drag=false)
   &SolveForcesAndVelocity(Clamp Velocity=false, Limit Acceleration=false)
   &ParticleState()
   &ScaleSpriteSize(Scale Factor=&Vector2DFromFloat(Value=&FloatFromCurve(FloatCurve=DataInterface() {propertyValues {Curve {keys [<at 0 0>,<at 0.2 1>,<at 1 0>]}}})))
  ]
  rendererClasses [&NiagaraSpriteRendererProperties]
  renderers [{rendererClass=&NiagaraSpriteRendererProperties}]
 }
 Droplets_Ribbon {
  emitterUpdateScript [
   &EmitterState(Life Cycle Mode=::NewEnumerator1, Loop Behavior=::NewEnumerator1, Loop Delay=0.3, Loop Duration=0.15, UseLoopDelay=true)
   &SpawnRate(SpawnRate=17, Use Spawn Probability=true)
  ]
  particleSpawnScript [
   &InitializeParticle(
    Color Mode=::NewEnumerator0
    Lifetime=&RandomRangeFloat(Minimum=0.333, Maximum=1)
    Sprite Size=<1.5 3>
    Sprite Size Mode=::NewEnumerator4
    UsePositionOffset=false
    Write Material Random=true
   )
   &InheritVelocity(Module.Limit Speed=true, Velocity Scale=<0.7 0.7 1>)
  ]
  particleUpdateScript [
   &GravityForce(Gravity=&RandomRangeVector(Minimum=<0 0 -444>, Maximum=<1 1 -666>))
   &Drag(Drag=3, Use Linear Drag=true, Use Rotational Drag=false)
   &SolveForcesAndVelocity(Clamp Velocity=false, Limit Acceleration=false)
   &ParticleState()
   &ScaleSpriteSize(Scale Factor=&Vector2DFromFloat(Value=&FloatFromCurve(FloatCurve=DataInterface() {propertyValues {Curve {keys [<at 0 1>,<at 1 0>]}}})))
  ]
  rendererClasses [&NiagaraSpriteRendererProperties]
  renderers [{rendererClass=&NiagaraSpriteRendererProperties}]
 }
 Droplets_Splash {
  emitterUpdateScript [
   &EmitterState()
   &SpawnBurst_Instantaneous(Spawn Count=&RandomRangeInt(Minimum=2, Maximum=4), Use Spawn Probability=false)
  ]
  particleSpawnScript [
   &InitializeParticle(
    Lifetime=&RandomRangeFloat(Minimum=0.333, Maximum=1)
    Sprite Size=<1.5 3>
    Sprite Size Mode=::NewEnumerator4
    UsePositionOffset=false
    Write Material Random=true
   )
   &AddVelocityInCone(
    Cone Axis=<1 -1 6>
    Cone Axis Coordinate Space=::Simulation
    Module.Use Velocity Falloff On Cone Axis=false
    Velocity Strength=&RandomRangeFloat(Minimum=55, Maximum=135)
   )
  ]
  particleUpdateScript [
   &GravityForce(Gravity=<0 0 -222>)
   &Drag(Use Linear Drag=true, Use Rotational Drag=false)
   &SolveForcesAndVelocity(Clamp Velocity=false, Limit Acceleration=false)
   &ParticleState()
   &ScaleSpriteSize(Scale Factor=&Vector2DFromFloat(Value=&FloatFromCurve(FloatCurve=DataInterface() {propertyValues {Curve {keys [<at 0 1>,<at 1 0>]}}})))
  ]
  rendererClasses [&NiagaraSpriteRendererProperties]
  renderers [{rendererClass=&NiagaraSpriteRendererProperties}]
 }
 Ribbon {
  emitterUpdateScript [
   &EmitterState(Life Cycle Mode=::NewEnumerator1, Loop Behavior=::NewEnumerator1, Loop Delay=0.3, Loop Duration=0.15, UseLoopDelay=true)
   &FN_SpawnRibbon_v2(SpawnRate=100)
  ]
  particleSpawnScript [
   &InitializeParticle(Color Mode=::NewEnumerator0, Lifetime=0.1, UsePositionOffset=false, Write Material Random=true)
   /Game/Effects/Niagara/Modules/Ribbons/FN_InitializeRibbon_v5(Lifetime=0.333, Ribbon Width=1)
   &ApplyInitialForces()
   &FN_UpdateRibbon()
  ]
  particleUpdateScript [
   &ParticleState()
   &GravityForce(Gravity=<0 0 -444>)
   &SolveForcesAndVelocity(Clamp Velocity=false, Limit Acceleration=false)
   &ScaleColor(
    Scale Alpha=&FloatFromCurve(FloatCurve=DataInterface() {propertyValues {Curve {keys [
         <at 0 0>
         {arriveTangent=-3.1876, interpMode=RCIM_Cubic, leaveTangent=-3.1876, tangentMode=RCTM_User, time=0.1, value=0.9}
         {arriveTangent=-0.1462, interpMode=RCIM_Cubic, leaveTangent=-0.1462, tangentMode=RCTM_User, time=1, value=0}
        ]}}})
    ScaleA=true
    ScaleRGB=true
   )
  ]
  rendererClasses [&NiagaraRibbonRendererProperties]
  renderers [{rendererClass=&NiagaraRibbonRendererProperties}]
 }
 Splash {
  emitterUpdateScript [
   &EmitterState()
   &SpawnBurst_Instantaneous(Spawn Count=1, Use Spawn Probability=false)
  ]
  particleSpawnScript [
   &InitializeParticle(
    Lifetime=0.45
    Sprite Rotation Angle=0.125
    Sprite Rotation Mode=::NewEnumerator1
    Sprite Size Mode=::NewEnumerator3
    Uniform Sprite Size=15
    UsePositionOffset=false
    Write Material Random=true
   )
   &AddVelocityInCone(
    Cone Angle=0
    Cone Axis=<1 -1 3>
    Cone Axis Coordinate Space=::Simulation
    Module.Use Velocity Falloff On Cone Axis=false
    Velocity Strength=77
   )
  ]
  particleUpdateScript [
   &Drag(Drag=3, Use Linear Drag=true, Use Rotational Drag=false)
   &SolveForcesAndVelocity(Clamp Velocity=false, Limit Acceleration=false)
   &ParticleState()
   &ScaleColor(
    Scale Alpha=&FloatFromCurve(FloatCurve=DataInterface() {propertyValues {Curve {keys [<at 0 1>,<at 1 0>]}}})
    ScaleA=true
    ScaleRGB=true
   )
   &CameraOffset()
   &ScaleSpriteSize(Scale Factor=&Vector2DFromFloat(Value=&FloatFromCurve(FloatCurve=DataInterface() {propertyValues {Curve {keys [
          {arriveTangent=3.2525, interpMode=RCIM_Cubic, leaveTangent=3.2525, tangentMode=RCTM_User, time=0, value=0.25}
          {interpMode=RCIM_Cubic, tangentMode=RCTM_User, time=0.5, value=1}
         ]}}})))
  ]
  rendererClasses [&NiagaraSpriteRendererProperties]
  renderers [{rendererClass=&NiagaraSpriteRendererProperties}]
 }
}
summary {
 systemName=NS_Emote_Breakthrough
 emitters [
  {emitterName=Splash, rendererClasses [&NiagaraSpriteRendererProperties]}
  {emitterName=Droplets_Splash, rendererClasses [&NiagaraSpriteRendererProperties]}
  {emitterName=Ribbon, rendererClasses [&NiagaraRibbonRendererProperties]}
  {emitterName=Droplets_Ribbon, rendererClasses [&NiagaraSpriteRendererProperties]}
  {emitterName=Droplets_Residual, rendererClasses [&NiagaraSpriteRendererProperties]}
 ]
 userVariables [User.RibbonIDOffset {type=NiagaraInt32}]
}
