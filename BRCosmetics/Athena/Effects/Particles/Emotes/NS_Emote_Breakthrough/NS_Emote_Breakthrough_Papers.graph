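# Niagara system for the Breakthrough emote that throws a small burst of mesh particles, evidently papers, into the air.
# Its single PapersPlease emitter spawns them once in a stretched sphere and launches them upward in a cone with random size, orientation and lifetime.
# Age-eased drag, randomized sideways and downward acceleration, and tumbling rotation shape the fall.
# A scale curve pops each mesh in quickly, then shrinks it to nothing.
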NiagaraSystem /BRCosmetics/Athena/Effects/Particles/Emotes/NS_Emote_Breakthrough/NS_Emote_Breakthrough_Papers
dependencies {usedDataInterfaces [&NiagaraDataInterfaceCurve], usedDynamicInputs [], usedModules [&SystemState]}
emitters {PapersPlease {
  emitterUpdateScript [
   &EmitterState()
   &SpawnBurst_Instantaneous(Spawn Count=7, Use Spawn Probability=false)
  ]
  particleSpawnScript [
   &InitializeParticle(
    Color=<rgba 0.6 0.3649 0.1242 0.3>
    Lifetime=&RandomRangeFloat(Minimum=1.7, Maximum=6.6)
    Mesh Scale=&Multiply_VectorByFloat(
     Vector=&RandomRangeVector(Minimum=<0.6 0.9 0.7>, Maximum=<0.8 0.9 0.9>)
     Float=&RandomRangeFloat(Minimum=0.11, Maximum=0.18)
    )
    Mesh Scale Mode=::NewEnumerator4
    Position Mode=::NewEnumerator0
    Write Material Random=true
   )
   &SphereLocation(
    MaskSpawnGroup=false
    Non Uniform Scale=<1 1 2>
    Override Seed=false
    Sphere Radius=5
    Surface Only=false
    UseNonUniformScale=true
    UseOffset=false
    UseSphereOrientation=false
   )
   &AddVelocityInCone(
    Cone Axis=<0 0 1>
    Cone Axis Coordinate Space=::World
    Module.Use Velocity Falloff On Cone Axis=false
    Velocity Distribution Along Cone Axis=0
    Velocity Strength=&RandomRangeFloat(Minimum=222, Maximum=666)
   )
   &InitialMeshOrientation(
    Orientation Vector=&RandomRangeVector(Minimum=<0 -0.5 -0.5>, Maximum=<0 0.5 -1>)
    Rotation Coordinate Space=::NewEnumerator0
    Use Orientation Vector=true
    Use Rotation Vector=true
   )
  ]
  particleUpdateScript [
   &Drag(
    Drag=&Lerp_Float(
     A=3
     B=2
     Alpha=&FloatFromCurve(FloatCurve=DataInterface() {propertyValues {Curve {keys [<at 0 0 RCIM_Cubic>, {interpMode=RCIM_Cubic, tangentMode=RCTM_Break, time=1, value=1}]}}})
    )
    Use Linear Drag=true
    Use Rotational Drag=false
   )
   &AccelerationForce(Acceleration=&RandomRangeVector(Minimum=<99 99 -99>, Maximum=<-99 -99 -111>))
   &ParticleState()
   &UpdateMeshOrientation(
    Rotation Coordinate Space=::NewEnumerator2
    Rotation Rate=&RandomRangeFloat(Minimum=0.4, Maximum=0.9)
    Rotation Vector=&NormalizeVector(Vector To Normalize=&RandomRangeVector(Minimum=<-1 -1 -0.2>, Maximum=<1 1 0.2>))
   )
   &ScaleMeshSize(Scale Factor=&VectorFromFloat(Value=&FloatFromCurve(FloatCurve=DataInterface() {propertyValues {Curve {keys [
          <at 0 0 RCIM_Cubic>
          {arriveTangentWeight=0.034, interpMode=RCIM_Cubic, leaveTangentWeight=0.3059, tangentMode=RCTM_User, time=0.015, value=1}
          {arriveTangent=-2.7218, interpMode=RCIM_Cubic, tangentMode=RCTM_Break, time=1, value=0}
         ]}}})))
   &SolveForcesAndVelocity(Clamp Velocity=false, Limit Acceleration=false)
   &SolveRotationalForcesAndVelocity()
  ]
  rendererClasses [&NiagaraMeshRendererProperties]
  renderers [{rendererClass=&NiagaraMeshRendererProperties, rendererIndex=0}]
 }}
summary {
 systemName=NS_Emote_Breakthrough_Papers
 emitters [{emitterName=PapersPlease, rendererClasses [&NiagaraMeshRendererProperties]}]
 userVariables []
}
